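# Niagara cosmetic effect for the M_MED_Citadel player body, a persistent pulsing plasma core, apparently on the chest, with glow.
# Plasma_Chest and Plasma_Inverted are mesh particles animated by curve-driven dynamic material parameters; the chest mesh also pulses in XY scale.
# Glow places four sprites by execution-index offsets, tinted by a user variant color with subtle alpha pulsing.
# Bursts spawn only on loop zero, with looping particle lifetimes until emitter deactivation.
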NiagaraSystem /BRCosmetics/Characters/Player/Male/Medium/Bodies/M_MED_Citadel/FX/NS_Citadel
dependencies {
 usedDataInterfaces [&NiagaraDataInterfaceCurve, &NiagaraDataInterfaceArrayFloat3]
 usedDynamicInputs []
 usedModules [&SystemState]
}
emitters {
 Glow {
  emitterUpdateScript [
   &EmitterState(Inactive Response=::NewEnumerator1, Life Cycle Mode=::NewEnumerator1, UseLoopDelay=false)
   &SpawnBurst_Instantaneous(Spawn Count=4, Spawn Probability=HlslExpression() {hlslExpression="Emitter.LoopCount == 0"}, Use Spawn Probability=true)
  ]
  particleSpawnScript [
   &InitializeParticle(
    Color=$User.Variant_Color_Plasma_Glow
    Sprite Size Mode=::NewEnumerator3
    Uniform Sprite Size=50
    UsePositionOffset=false
    Write Material Random=true
   ) {module=InitializeParticle001}
   &SpriteFacingAndAlignment(Alignment Coordinate Space=::Local, Sprite Alignment=<1 0.25 0>, WriteAlignment=true, WriteFacing=false)
   &SystemLocation(
    MaskSpawnGroup=false
    Offset=&SelectVectorFromArray(
     Direct Array Index=&ReturnExecIndex()
     Vector Selection Array=DataInterface() {propertyValues {FloatData [<0 -1.5 -8>, <0 -3 4>, <3.5 -1.5 0>, <-3.5 -1.5 0>]}}
    )
   )
  ]
  particleUpdateScript [
   &ParticleState(
    Kill Particles When Lifetime Has Elapsed=false
    Let Infinitely Lived Particles Die When Emitter Deactivates=true
    Loop Particles Lifetime=true
   )
   &CameraOffset(Camera Offset Amount=2.5)
   &ScaleColor(
    Color Value To Scale=$User.Variant_Color_Plasma_Glow
    Scale Alpha=&FloatFromCurve(FloatCurve=DataInterface() {propertyValues {Curve {keys [
         {arriveTangent=-0.4837, interpMode=RCIM_Cubic, leaveTangent=-0.4837, tangentMode=RCTM_User, time=0, value=1}
         {arriveTangent=0.862, interpMode=RCIM_Cubic, leaveTangent=0.862, tangentMode=RCTM_User, time=0.5, value=1.025}
         {arriveTangent=0.8247, interpMode=RCIM_Cubic, leaveTangent=0.8247, tangentMode=RCTM_User, time=1, value=1}
        ]}}})
    ScaleA=true
    ScaleRGB=false
   )
  ]
  rendererClasses [&NiagaraSpriteRendererProperties]
  renderers [{rendererClass=&NiagaraSpriteRendererProperties}]
 }
 Plasma_Chest {
  emitterUpdateScript [
   &EmitterState(Life Cycle Mode=::NewEnumerator1, UseLoopDelay=false)
   &SpawnBurst_Instantaneous(Spawn Probability=HlslExpression() {hlslExpression="Emitter.LoopCount == 0"}, Use Spawn Probability=true, UseLoopCountLimit=false)
  ]
  particleSpawnScript [
   &InitializeParticle(Mesh Scale=<0.65 0.7 1>, Mesh Scale Mode=::NewEnumerator4, UsePositionOffset=false, Write Material Random=true)
   &InitialMeshOrientation(Mesh Orientation Mode=::NewEnumerator2, Rotation=<0.28 0 0>, Use Rotation Vector=true)
   &SystemLocation(MaskSpawnGroup=false, Offset=<0 -1 -0.5>)
  ]
  particleUpdateScript [
   &ParticleState(
    Kill Particles When Lifetime Has Elapsed=false
    Let Infinitely Lived Particles Die When Emitter Deactivates=true
    Loop Particles Lifetime=true
   )
   &CameraOffset(Camera Offset Amount=1)
   &DynamicMaterialParameters(
    Index 0 Param 1=&FloatFromCurve(
     FloatCurve=DataInterface() {propertyValues {Curve {keys [
         {interpMode=RCIM_Cubic, tangentMode=RCTM_User, time=0, value=0.05}
         {arriveTangent=-0.0008, interpMode=RCIM_Cubic, leaveTangent=-0.0008, tangentMode=RCTM_User, time=0.1474, value=0.0316}
         {arriveTangent=-0.0621, interpMode=RCIM_Cubic, leaveTangent=-0.0621, tangentMode=RCTM_User, time=0.2888, value=0.1386}
         {arriveTangent=-0.1732, interpMode=RCIM_Cubic, leaveTangent=-0.1732, tangentMode=RCTM_User, time=0.4007, value=0.05}
         {interpMode=RCIM_Cubic, tangentMode=RCTM_User, time=0.4826, value=0.0427}
         {arriveTangent=-0.0057, interpMode=RCIM_Cubic, leaveTangent=-0.0057, tangentMode=RCTM_User, time=0.593, value=0.037}
         {arriveTangent=0.0385, interpMode=RCIM_Cubic, leaveTangent=0.0385, time=0.8161, value=0.1381}
         {arriveTangent=-0.0113, interpMode=RCIM_Cubic, leaveTangent=-0.0113, tangentMode=RCTM_User, time=0.9311, value=0.05}
         {tangentMode=RCTM_User, time=1, value=0.05}
        ]}}}
     CurveIndex=$Emitter.NormalizedLoopAge
    )
    Index 0 Param 2=&FloatFromCurve(
     FloatCurve=DataInterface() {propertyValues {Curve {keys [
         {interpMode=RCIM_Cubic, tangentMode=RCTM_User, time=0, value=150}
         {arriveTangent=-0.0008, interpMode=RCIM_Cubic, leaveTangent=-0.0008, tangentMode=RCTM_User, time=0.0479, value=150}
         {arriveTangent=174.1042, interpMode=RCIM_Cubic, leaveTangent=174.1042, time=0.1648, value=200}
         {arriveTangent=-484.7725, interpMode=RCIM_Cubic, leaveTangent=-484.7725, tangentMode=RCTM_User, time=0.2573, value=186.446}
         {interpMode=RCIM_Cubic, tangentMode=RCTM_User, time=0.3963, value=230}
         {arriveTangent=372.7463, interpMode=RCIM_Cubic, leaveTangent=372.7463, tangentMode=RCTM_User, time=0.6102, value=165.0835}
         {arriveTangent=-82.9196, interpMode=RCIM_Cubic, leaveTangent=-82.9196, time=0.7076, value=200}
         {arriveTangent=-0.0113, interpMode=RCIM_Cubic, leaveTangent=-0.0113, tangentMode=RCTM_User, time=0.7921, value=150}
         {tangentMode=RCTM_User, time=1, value=150}
        ]}}}
     Scale Curve=0.5
    )
    Module.Param0WriteEnabled=true
    Module.Param1WriteEnabled=true
    Module.Param2WriteEnabled=false
    Module.Param3WriteEnabled=false
   )
   &ScaleMeshSize(Scale Factor=&MakeVector(
     X=&FloatFromCurve(FloatCurve=DataInterface() {propertyValues {Curve {keys [
          {interpMode=RCIM_Cubic, tangentMode=RCTM_User, time=0, value=0.95}
          {interpMode=RCIM_Cubic, tangentMode=RCTM_User, time=0.5, value=1}
          {tangentMode=RCTM_User, time=1, value=0.95}
         ]}}})
     Y=&FloatFromCurve(FloatCurve=DataInterface() {propertyValues {Curve {keys [
          {interpMode=RCIM_Cubic, tangentMode=RCTM_User, time=0, value=0.95}
          {interpMode=RCIM_Cubic, tangentMode=RCTM_User, time=0.5, value=1}
          {tangentMode=RCTM_User, time=1, value=0.95}
         ]}}})
     Z=1
    ))
  ]
  rendererClasses [&NiagaraMeshRendererProperties]
  renderers [{rendererClass=&NiagaraMeshRendererProperties, rendererIndex=0}]
 }
 Plasma_Inverted {
  emitterUpdateScript [
   &EmitterState(Life Cycle Mode=::NewEnumerator1, UseLoopDelay=false)
   &SpawnBurst_Instantaneous(Spawn Probability=HlslExpression() {hlslExpression="Emitter.LoopCount == 0"}, Use Spawn Probability=true, UseLoopCountLimit=false)
  ]
  particleSpawnScript [
   &InitializeParticle(Mesh Scale Mode=::NewEnumerator3, Mesh Uniform Scale=1.8, UsePositionOffset=false, Write Material Random=true)
   &InitialMeshOrientation(Mesh Orientation Mode=::NewEnumerator2, Orientation Vector=<0 1 0>, Use Rotation Vector=true)
   &SystemLocation(MaskSpawnGroup=false, Offset=<0 -1.75 0>)
  ]
  particleUpdateScript [
   &ParticleState(
    Kill Particles When Lifetime Has Elapsed=false
    Let Infinitely Lived Particles Die When Emitter Deactivates=true
    Loop Particles Lifetime=true
   )
   &DynamicMaterialParameters(
    Index 0 Param 1=&FloatFromCurve(
     FloatCurve=DataInterface() {propertyValues {Curve {keys [
         {interpMode=RCIM_Cubic, tangentMode=RCTM_User, time=0, value=0.05}
         {arriveTangent=-0.0008, interpMode=RCIM_Cubic, leaveTangent=-0.0008, tangentMode=RCTM_User, time=0.1474, value=0.0316}
         {arriveTangent=-0.0621, interpMode=RCIM_Cubic, leaveTangent=-0.0621, tangentMode=RCTM_User, time=0.2888, value=0.1386}
         {arriveTangent=-0.1732, interpMode=RCIM_Cubic, leaveTangent=-0.1732, tangentMode=RCTM_User, time=0.4007, value=0.05}
         {interpMode=RCIM_Cubic, tangentMode=RCTM_User, time=0.4826, value=0.0427}
         {arriveTangent=-0.0057, interpMode=RCIM_Cubic, leaveTangent=-0.0057, tangentMode=RCTM_User, time=0.593, value=0.037}
         {arriveTangent=0.0385, interpMode=RCIM_Cubic, leaveTangent=0.0385, time=0.8161, value=0.1381}
         {arriveTangent=-0.0113, interpMode=RCIM_Cubic, leaveTangent=-0.0113, tangentMode=RCTM_User, time=0.9311, value=0.05}
         {tangentMode=RCTM_User, time=1, value=0.05}
        ]}}}
     CurveIndex=$Emitter.NormalizedLoopAge
    )
    Index 0 Param 2=&FloatFromCurve(FloatCurve=DataInterface() {propertyValues {Curve {keys [
         {interpMode=RCIM_Cubic, tangentMode=RCTM_User, time=0, value=150}
         {arriveTangent=-0.0008, interpMode=RCIM_Cubic, leaveTangent=-0.0008, tangentMode=RCTM_User, time=0.0479, value=150}
         {arriveTangent=174.1042, interpMode=RCIM_Cubic, leaveTangent=174.1042, time=0.1648, value=200}
         {arriveTangent=-484.7725, interpMode=RCIM_Cubic, leaveTangent=-484.7725, tangentMode=RCTM_User, time=0.2573, value=186.446}
         {interpMode=RCIM_Cubic, tangentMode=RCTM_User, time=0.3963, value=230}
         {arriveTangent=372.7463, interpMode=RCIM_Cubic, leaveTangent=372.7463, tangentMode=RCTM_User, time=0.6102, value=165.0835}
         {arriveTangent=-82.9196, interpMode=RCIM_Cubic, leaveTangent=-82.9196, time=0.7076, value=200}
         {arriveTangent=-0.0113, interpMode=RCIM_Cubic, leaveTangent=-0.0113, tangentMode=RCTM_User, time=0.7921, value=150}
         {tangentMode=RCTM_User, time=1, value=150}
        ]}}})
    Module.Param0WriteEnabled=true
    Module.Param1WriteEnabled=true
    Module.Param2WriteEnabled=false
    Module.Param3WriteEnabled=false
   )
  ]
  rendererClasses [&NiagaraMeshRendererProperties]
  renderers [{rendererClass=&NiagaraMeshRendererProperties, rendererIndex=0}]
 }
}
summary {
 systemName=NS_Citadel
 emitters [
  {emitterName=Plasma_Chest, rendererClasses [&NiagaraMeshRendererProperties]}
  {emitterName=Plasma_Inverted, rendererClasses [&NiagaraMeshRendererProperties]}
  {emitterName=Glow, rendererClasses [&NiagaraSpriteRendererProperties]}
 ]
 userVariables [
  User.SK_Mesh {
   type=NiagaraDataInterfaceSkeletalMesh
   defaultValue {
    struct=NiagaraExt_VariableValue_DataInterface
    value {dataInterface=&self:NiagaraDataInterfaceSkeletalMesh_0}
   }
  }
  User.Variant_Color_Plasma_Chest {type=LinearColor, description="", defaultValue {struct=LinearColor, value=<rgba 23.875 13.75 25 0>}}
  User.Variant_Color_Plasma_Glow {type=LinearColor, description="", defaultValue {struct=LinearColor, value=<rgba 0.96 0.6 1 0.05>}}
  User.Variant_Color_Plasma_Inverted {type=LinearColor, description="", defaultValue {struct=LinearColor, value=<rgba 47.75 27.5 50 0>}}
 ]
}
