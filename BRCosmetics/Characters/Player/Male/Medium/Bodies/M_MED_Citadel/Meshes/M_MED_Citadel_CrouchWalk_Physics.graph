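# Physics asset for the crouch-walk variant of the male medium Citadel body, covering its dangling loincloth and straps.
# Capsules on pelvis, spine_01, thighs and calves form a lower-body proxy with free angular joints.
# Left/right loincloth and strap chains hang from pelvis through tiny MSTR anchor bodies, which are angularly free for straps only.
# Chain links carry per-joint swing limits and zero twist limit, with four segments per loincloth panel and one or two per strap.
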PhysicsAsset /BRCosmetics/Characters/Player/Male/Medium/Bodies/M_MED_Citadel/Meshes/M_MED_Citadel_CrouchWalk_Physics
bodies {
 calf_l [{shapeType=Capsule, length=28.2006, radius=6.2504, shapeName=calf_l_capsule, center=<-14.1936 0 0>, rotation=<rot 90 -179.156 -179.156>}]
 calf_r [{shapeType=Capsule, length=28.2006, radius=6.2504, shapeName=calf_l_capsule, center=<14.1936 0 0>, rotation=<rot 90 -179.8417 0.1583>}]
 dyn_loinClothBack_l_01 [{shapeType=Capsule, length=0.1284, radius=3.6679, shapeName=dyn_loinClothBack_l_01_capsule, center=<1.9849 -3.3 0>, rotation=<rot 90 0 0>}]
 dyn_loinClothBack_l_02 [{shapeType=Capsule, length=7.96, radius=4.7289, shapeName=dyn_loinClothBack_l_02_capsule, center=<4.1597 -3.3 0>, rotation=<rot 90 -180 180>}]
 dyn_loinClothBack_l_03 [{shapeType=Capsule, length=7.8694, radius=4.2689, shapeName=dyn_loinClothBack_l_03_capsule, center=<3.9765 -3.3 0>, rotation=<rot 90 0 0>}]
 dyn_loinClothBack_l_04 [{shapeType=Capsule, length=5.51, radius=4.0914, shapeName=dyn_loinClothBack_l_04_capsule, center=<3.5349 -3.3 0>, rotation=<rot 90 180 180>}]
 dyn_loinClothBack_l_MSTR [{shapeType=Capsule, length=1, radius=1, shapeName=dyn_loinClothBack_l_MSTR_capsule}]
 dyn_loinClothBack_r_01 [{
   shapeType=Capsule
   length=0.1284
   radius=3.6679
   shapeName=dyn_loinClothBack_r_01_capsule
   center=<1.9849 -3.3 0>
   rotation=<rot 90 179.578 179.578>
  }]
 dyn_loinClothBack_r_02 [{shapeType=Capsule, length=7.0718, radius=5.2721, shapeName=dyn_loinClothBack_r_02_capsule, center=<3.8358 -3.3 0>, rotation=<rot 90 0 0>}]
 dyn_loinClothBack_r_03 [{shapeType=Capsule, length=7.7522, radius=4.4992, shapeName=dyn_loinClothBack_r_03_capsule, center=<3.9765 -3.3 0>, rotation=<rot 90 180 180>}]
 dyn_loinClothBack_r_04 [{shapeType=Capsule, length=5.145, radius=4.4564, shapeName=dyn_loinClothBack_r_04_capsule, center=<3.5349 -3.3 0>, rotation=<rot 90 180 180>}]
 dyn_loinClothBack_r_MSTR [{shapeType=Capsule, length=1, radius=1, shapeName=dyn_loinClothBack_r_MSTR_capsule}]
 dyn_loinClothFront_l_01 [{shapeType=Capsule, length=1.6, radius=3.3, shapeName=dyn_loinClothFront_l_01_capsule, center=<1.9849 2.1 -2.2>, rotation=<rot 90 0 0>}]
 dyn_loinClothFront_l_02 [{shapeType=Capsule, length=4, radius=3.3, shapeName=dyn_loinClothFront_l_02_capsule, center=<3.8 2.1 -2.2>, rotation=<rot 90 0 0>}]
 dyn_loinClothFront_l_03 [{shapeType=Capsule, length=8.6654, radius=2.0262, shapeName=dyn_loinClothFront_l_03_capsule, center=<3.4103 1.3 -1.3>, rotation=<rot 90 0 0>}]
 dyn_loinClothFront_l_04 [{shapeType=Capsule, length=10.8171, radius=2.986, shapeName=dyn_loinClothFront_l_04_capsule, center=<4.4172 1.7 -1.5>, rotation=<rot 90 0 0>}]
 dyn_loinClothFront_l_MSTR [{shapeType=Capsule, length=1, radius=1, shapeName=dyn_loinClothFront_l_MSTR_capsule}]
 dyn_loinClothFront_r_01 [{shapeType=Capsule, length=1.6, radius=3.3, shapeName=dyn_loinClothFront_r_01_capsule, center=<1.9849 2.1 2.2>, rotation=<rot 90 0 0>}]
 dyn_loinClothFront_r_02 [{shapeType=Capsule, length=4, radius=3.3, shapeName=dyn_loinClothFront_r_02_capsule, center=<3.8 2.1 2.2>, rotation=<rot 90 0 0>}]
 dyn_loinClothFront_r_03 [{shapeType=Capsule, length=8.65, radius=2, shapeName=dyn_loinClothFront_r_03_capsule, center=<3.4103 1.3 1.3>, rotation=<rot 90 0 0>}]
 dyn_loinClothFront_r_04 [{shapeType=Capsule, length=10.8171, radius=2.986, shapeName=dyn_loinClothFront_r_04_capsule, center=<4.4172 1.7 1.5>, rotation=<rot 90 0 0>}]
 dyn_loinClothFront_r_MSTR [{shapeType=Capsule, length=1, radius=1, shapeName=dyn_loinClothFront_r_MSTR_capsule}]
 dyn_strapA_l_01 [{
   shapeType=Capsule
   length=8.7802
   radius=1.6677
   shapeName=dyn_strapA_l_01_capsule
   center=<11.642 -1.2 0.4>
   rotation=<rot 41.399 170.2753 -177.7415>
  }]
 dyn_strapA_l_MSTR [{shapeType=Capsule, length=0.505, radius=0.505, shapeName=dyn_strapA_l_MSTR_capsule, rotation=<rot 0 -90 0>}]
 dyn_strapA_r_01 [{
   shapeType=Capsule
   length=8.572
   radius=1.9352
   shapeName=dyn_strapA_r_01_capsule
   center=<12.1561 -0.8 -1>
   rotation=<rot 9.9656 -98.771 -38.6143>
  }]
 dyn_strapA_r_MSTR [{shapeType=Capsule, length=1, radius=1, shapeName=dyn_strapA_r_MSTR_capsule, rotation=<rot 0 -90 0>}]
 dyn_strapBack_l_01 [{shapeType=Capsule, length=5, radius=2, shapeName=dyn_strapBack_l_01_capsule, center=<3.073 2.283 0>, rotation=<rot 90 0 0>}]
 dyn_strapBack_l_02 [{shapeType=Capsule, length=5, radius=2, shapeName=dyn_strapBack_l_02_capsule, center=<3.073 2.283 0>, rotation=<rot 90 0 0>}]
 dyn_strapBack_l_MSTR [{shapeType=Capsule, length=1, radius=1, shapeName=dyn_strapBack_l_MSTR_capsule}]
 dyn_strapBack_r_01 [{
   shapeType=Capsule
   length=4.4069
   radius=1.5931
   shapeName=dyn_strapBack_r_01_capsule
   center=<2.2884 1.7 0>
   rotation=<rot 90 180 -178.4099>
  }]
 dyn_strapBack_r_02 [{shapeType=Capsule, length=6.989, radius=1.7473, shapeName=dyn_strapBack_r_02_capsule, center=<2.3323 1.4 0>, rotation=<rot 90 0 9.678>}]
 dyn_strapBack_r_MSTR [{shapeType=Capsule, length=1, radius=1, shapeName=dyn_strapBack_r_MSTR_capsule}]
 dyn_strapC_l_01 [{
   shapeType=Capsule
   length=8.4437
   radius=2.061
   shapeName=dyn_strapC_l_01_capsule
   center=<12.2261 0.9 -0.5>
   rotation=<rot 39.2168 -9.6731 -0.6595>
  }]
 dyn_strapC_l_MSTR [{shapeType=Capsule, length=0.505, radius=0.505, shapeName=dyn_strapC_l_MSTR_capsule, rotation=<rot 0 -90 0>}]
 dyn_strapC_r_01 [{
   shapeType=Capsule
   length=10.1462
   radius=1.7429
   shapeName=dyn_strapC_r_01_capsule
   center=<12.5185 1 0.5>
   rotation=<rot -8.3647 -91.9817 27.1323>
  }]
 dyn_strapC_r_MSTR [{shapeType=Capsule, length=1, radius=1, shapeName=dyn_strapC_r_MSTR_capsule, rotation=<rot 0 -90 0>}]
 dyn_strapFront_l_01 [{shapeType=Capsule, length=5.2243, radius=1.7757, shapeName=dyn_strapFront_l_01_capsule, center=<1.7368 1.2 0>, rotation=<rot 90 0 0>}]
 dyn_strapFront_l_02 [{shapeType=Capsule, length=5.1276, radius=1.8724, shapeName=dyn_strapFront_l_02_capsule, center=<1.7368 1.2 0>, rotation=<rot 90 0 0>}]
 dyn_strapFront_l_MSTR [{shapeType=Capsule, length=1, radius=1, shapeName=dyn_strapFront_l_MSTR_capsule}]
 dyn_strapFront_r_01 [{
   shapeType=Capsule
   length=7.0475
   radius=1.4656
   shapeName=dyn_strapFront_r_01_capsule
   center=<3.8176 1.173 0>
   rotation=<rot 90 179.5524 -175.1933>
  }]
 dyn_strapFront_r_02 [{
   shapeType=Capsule
   length=8.6436
   radius=2.1124
   shapeName=dyn_strapFront_r_02_capsule
   center=<2.8392 1.15 -0.3>
   rotation=<rot 83.9457 -3.7952 2.0033>
  }]
 dyn_strapFront_r_MSTR [{shapeType=Capsule, length=1, radius=1, shapeName=dyn_strapFront_r_MSTR_capsule}]
 pelvis [
  {shapeType=Capsule, length=13.1838, radius=10.1966, shapeName=pelvis_capsule, center=<0 -0.8234 0>}
  {shapeType=Capsule, length=12.414, radius=9.7779, shapeName=pelvis_capsule, center=<-3.905 -1.6865 0>}
 ]
 spine_01 [{shapeType=Capsule, length=0.505, radius=0.505, shapeName=spine_01_capsule}]
 thigh_l [
  {
   shapeType=Capsule
   length=31.9853
   radius=8.1946
   shapeName=thigh_l_capsule
   center=<-14.4147 2.117 -1.333>
   rotation=<rot 88.1968 176.6878 -179.227>
  }
  {
   shapeType=Capsule
   length=22.4543
   radius=7.5457
   shapeName=thigh_l_capsule
   center=<-15.2209 -0.6982 -0.0363>
   rotation=<rot 90 180 -172.5581>
  }
 ]
 thigh_r [
  {
   shapeType=Capsule
   length=33.2153
   radius=8.2741
   shapeName=thigh_l_capsule
   center=<14.4147 -2.117 1.333>
   rotation=<rot 88.1968 176.6878 0.773>
  }
  {shapeType=Capsule, length=22.4543, radius=7.5457, shapeName=thigh_l_capsule, center=<15.1696 0.3055 0.0124>, rotation=<rot 90 0 -172.5581>}
 ]
}
constraints [
 {
  bone1Name=spine_01
  bone2Name=pelvis
  swing1LimitDegrees=45
  swing1Motion=Free
  swing2LimitDegrees=45
  swing2Motion=Free
  twistLimitDegrees=45
  twistMotion=Free
 }
 {
  bone1Name=thigh_l
  bone2Name=pelvis
  swing1LimitDegrees=45
  swing1Motion=Free
  swing2LimitDegrees=45
  swing2Motion=Free
  twistLimitDegrees=45
  twistMotion=Free
 }
 {
  bone1Name=calf_l
  bone2Name=thigh_l
  swing1LimitDegrees=45
  swing1Motion=Free
  swing2LimitDegrees=45
  swing2Motion=Free
  twistLimitDegrees=45
  twistMotion=Free
 }
 {bone1Name=dyn_loinClothBack_l_MSTR, bone2Name=pelvis, swing1LimitDegrees=45, swing2LimitDegrees=20, twistLimitDegrees=0}
 {bone1Name=dyn_loinClothBack_l_01, bone2Name=dyn_loinClothBack_l_MSTR, swing1LimitDegrees=45, swing2LimitDegrees=20, twistLimitDegrees=0}
 {bone1Name=dyn_loinClothBack_l_02, bone2Name=dyn_loinClothBack_l_01, swing1LimitDegrees=45, swing2LimitDegrees=20, twistLimitDegrees=0}
 {bone1Name=dyn_loinClothBack_l_03, bone2Name=dyn_loinClothBack_l_02, swing1LimitDegrees=45, swing2LimitDegrees=20, twistLimitDegrees=0}
 {bone1Name=dyn_loinClothBack_l_04, bone2Name=dyn_loinClothBack_l_03, swing1LimitDegrees=45, swing2LimitDegrees=20, twistLimitDegrees=0}
 {bone1Name=dyn_loinClothBack_r_MSTR, bone2Name=pelvis, swing1LimitDegrees=45, swing2LimitDegrees=20, twistLimitDegrees=0}
 {bone1Name=dyn_loinClothBack_r_01, bone2Name=dyn_loinClothBack_r_MSTR, swing1LimitDegrees=45, swing2LimitDegrees=20, twistLimitDegrees=0}
 {bone1Name=dyn_loinClothBack_r_02, bone2Name=dyn_loinClothBack_r_01, swing1LimitDegrees=45, swing2LimitDegrees=20, twistLimitDegrees=0}
 {bone1Name=dyn_loinClothBack_r_03, bone2Name=dyn_loinClothBack_r_02, swing1LimitDegrees=45, swing2LimitDegrees=20, twistLimitDegrees=0}
 {bone1Name=dyn_loinClothBack_r_04, bone2Name=dyn_loinClothBack_r_03, swing1LimitDegrees=45, swing2LimitDegrees=20, twistLimitDegrees=0}
 {bone1Name=dyn_loinClothFront_l_MSTR, bone2Name=pelvis, swing1LimitDegrees=45, swing2LimitDegrees=20, twistLimitDegrees=0}
 {bone1Name=dyn_loinClothFront_l_01, bone2Name=dyn_loinClothFront_l_MSTR, swing1LimitDegrees=45, swing2LimitDegrees=20, twistLimitDegrees=0}
 {bone1Name=dyn_loinClothFront_l_02, bone2Name=dyn_loinClothFront_l_01, swing1LimitDegrees=45, swing2LimitDegrees=20, twistLimitDegrees=0}
 {bone1Name=dyn_loinClothFront_l_03, bone2Name=dyn_loinClothFront_l_02, swing1LimitDegrees=45, swing2LimitDegrees=20, twistLimitDegrees=0}
 {bone1Name=dyn_loinClothFront_l_04, bone2Name=dyn_loinClothFront_l_03, swing1LimitDegrees=45, swing2LimitDegrees=20, twistLimitDegrees=0}
 {bone1Name=dyn_loinClothFront_r_MSTR, bone2Name=pelvis, swing1LimitDegrees=45, swing2LimitDegrees=20, twistLimitDegrees=0}
 {bone1Name=dyn_loinClothFront_r_01, bone2Name=dyn_loinClothFront_r_MSTR, swing1LimitDegrees=45, swing2LimitDegrees=20, twistLimitDegrees=0}
 {bone1Name=dyn_loinClothFront_r_02, bone2Name=dyn_loinClothFront_r_01, swing1LimitDegrees=45, swing2LimitDegrees=20, twistLimitDegrees=0}
 {bone1Name=dyn_loinClothFront_r_03, bone2Name=dyn_loinClothFront_r_02, swing1LimitDegrees=45, swing2LimitDegrees=20, twistLimitDegrees=0}
 {bone1Name=dyn_loinClothFront_r_04, bone2Name=dyn_loinClothFront_r_03, swing1LimitDegrees=45, swing2LimitDegrees=20, twistLimitDegrees=0}
 {
  bone1Name=calf_r
  bone2Name=thigh_r
  swing1LimitDegrees=45
  swing1Motion=Free
  swing2LimitDegrees=45
  swing2Motion=Free
  twistLimitDegrees=45
  twistMotion=Free
 }
 {
  bone1Name=thigh_r
  bone2Name=pelvis
  swing1LimitDegrees=45
  swing1Motion=Free
  swing2LimitDegrees=45
  swing2Motion=Free
  twistLimitDegrees=45
  twistMotion=Free
 }
 {
  bone1Name=dyn_strapBack_l_MSTR
  bone2Name=pelvis
  swing1LimitDegrees=45
  swing1Motion=Free
  swing2LimitDegrees=45
  swing2Motion=Free
  twistLimitDegrees=45
  twistMotion=Free
 }
 {bone1Name=dyn_strapBack_l_01, bone2Name=dyn_strapBack_l_MSTR, swing1LimitDegrees=45, swing2LimitDegrees=20, twistLimitDegrees=0}
 {bone1Name=dyn_strapBack_l_02, bone2Name=dyn_strapBack_l_01, swing1LimitDegrees=45, swing2LimitDegrees=10, twistLimitDegrees=0}
 {
  bone1Name=dyn_strapFront_l_MSTR
  bone2Name=pelvis
  swing1LimitDegrees=45
  swing1Motion=Free
  swing2LimitDegrees=45
  swing2Motion=Free
  twistLimitDegrees=45
  twistMotion=Free
 }
 {bone1Name=dyn_strapFront_l_01, bone2Name=dyn_strapFront_l_MSTR, swing1LimitDegrees=60, swing2LimitDegrees=10, twistLimitDegrees=0}
 {bone1Name=dyn_strapFront_l_02, bone2Name=dyn_strapFront_l_01, swing1LimitDegrees=45, swing2LimitDegrees=45, twistLimitDegrees=0}
 {
  bone1Name=dyn_strapBack_r_MSTR
  bone2Name=pelvis
  swing1LimitDegrees=45
  swing1Motion=Free
  swing2LimitDegrees=45
  swing2Motion=Free
  twistLimitDegrees=45
  twistMotion=Free
 }
 {bone1Name=dyn_strapBack_r_01, bone2Name=dyn_strapBack_r_MSTR, swing1LimitDegrees=45, swing2LimitDegrees=20, twistLimitDegrees=0}
 {bone1Name=dyn_strapBack_r_02, bone2Name=dyn_strapBack_r_01, swing1LimitDegrees=45, swing2LimitDegrees=20, twistLimitDegrees=0}
 {
  bone1Name=dyn_strapFront_r_MSTR
  bone2Name=pelvis
  swing1LimitDegrees=45
  swing1Motion=Free
  swing2LimitDegrees=45
  swing2Motion=Free
  twistLimitDegrees=45
  twistMotion=Free
 }
 {bone1Name=dyn_strapFront_r_01, bone2Name=dyn_strapFront_r_MSTR, swing1LimitDegrees=60, swing2LimitDegrees=20, twistLimitDegrees=0}
 {bone1Name=dyn_strapFront_r_02, bone2Name=dyn_strapFront_r_01, swing1LimitDegrees=55, swing2LimitDegrees=20, twistLimitDegrees=0}
 {
  bone1Name=dyn_strapA_l_MSTR
  bone2Name=pelvis
  swing1LimitDegrees=45
  swing1Motion=Free
  swing2LimitDegrees=45
  swing2Motion=Free
  twistLimitDegrees=45
  twistMotion=Free
 }
 {bone1Name=dyn_strapA_l_01, bone2Name=dyn_strapA_l_MSTR, swing1LimitDegrees=45, swing2LimitDegrees=45, twistLimitDegrees=0}
 {
  bone1Name=dyn_strapC_l_MSTR
  bone2Name=pelvis
  swing1LimitDegrees=45
  swing1Motion=Free
  swing2LimitDegrees=45
  swing2Motion=Free
  twistLimitDegrees=45
  twistMotion=Free
 }
 {bone1Name=dyn_strapC_l_01, bone2Name=dyn_strapC_l_MSTR, swing1LimitDegrees=45, swing2LimitDegrees=45, twistLimitDegrees=0}
 {
  bone1Name=dyn_strapA_r_MSTR
  bone2Name=pelvis
  swing1LimitDegrees=45
  swing1Motion=Free
  swing2LimitDegrees=45
  swing2Motion=Free
  twistLimitDegrees=45
  twistMotion=Free
 }
 {bone1Name=dyn_strapA_r_01, bone2Name=dyn_strapA_r_MSTR, swing1LimitDegrees=45, swing2LimitDegrees=45, twistLimitDegrees=0}
 {
  bone1Name=dyn_strapC_r_MSTR
  bone2Name=pelvis
  swing1LimitDegrees=45
  swing1Motion=Free
  swing2LimitDegrees=45
  swing2Motion=Free
  twistLimitDegrees=45
  twistMotion=Free
 }
 {bone1Name=dyn_strapC_r_01, bone2Name=dyn_strapC_r_MSTR, swing1LimitDegrees=45, swing2LimitDegrees=45, twistLimitDegrees=0}
]
