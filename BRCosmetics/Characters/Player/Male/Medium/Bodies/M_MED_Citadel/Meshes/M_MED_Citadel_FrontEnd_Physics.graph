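# Physics asset for the front-end version of the M_MED_Citadel male body, set up for the dangling loincloth and straps.
# Capsule bodies cover the pelvis, spine_01, both thighs and left/right chains of dyn_ bones, each rooted at a small MSTR body constrained to the pelvis.
# Chain links have tuned swing limits and zero-degree twist limits.
# Strap masters, thighs and spine have free angular motion, while loincloth masters keep angular limits.
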PhysicsAsset /BRCosmetics/Characters/Player/Male/Medium/Bodies/M_MED_Citadel/Meshes/M_MED_Citadel_FrontEnd_Physics
bodies {
 dyn_loinClothBack_l_01 [{shapeType=Capsule, length=3.8784, radius=2.9179, shapeName=dyn_loinClothBack_l_01_capsule, center=<1.9849 -3.6 0>, rotation=<rot 90 0 0>}]
 dyn_loinClothBack_l_02 [{shapeType=Capsule, length=9.21, radius=3.4789, shapeName=dyn_loinClothBack_l_02_capsule, center=<4.1597 -3.5 0>, rotation=<rot 90 -180 180>}]
 dyn_loinClothBack_l_03 [{shapeType=Capsule, length=8.6194, radius=3.5189, shapeName=dyn_loinClothBack_l_03_capsule, center=<3.9765 -2.9 0>, rotation=<rot 90 0 0>}]
 dyn_loinClothBack_l_04 [{shapeType=Capsule, length=6.51, radius=3.0914, shapeName=dyn_loinClothBack_l_04_capsule, center=<4.781 -1.3 0>, rotation=<rot 90 180 180>}]
 dyn_loinClothBack_l_MSTR [{shapeType=Capsule, length=1, radius=1, shapeName=dyn_loinClothBack_l_MSTR_capsule}]
 dyn_loinClothBack_r_01 [{
   shapeType=Capsule
   length=2.8784
   radius=3.9179
   shapeName=dyn_loinClothBack_r_01_capsule
   center=<1.9849 -3.6 0>
   rotation=<rot 90 179.8417 179.8417>
  }]
 dyn_loinClothBack_r_02 [{shapeType=Capsule, length=8.0718, radius=4.2721, shapeName=dyn_loinClothBack_r_02_capsule, center=<3.8358 -3.5 0>, rotation=<rot 90 0 0>}]
 dyn_loinClothBack_r_03 [{shapeType=Capsule, length=8.0022, radius=4.2492, shapeName=dyn_loinClothBack_r_03_capsule, center=<3.9765 -2.9 0>, rotation=<rot 90 180 180>}]
 dyn_loinClothBack_r_04 [{shapeType=Capsule, length=6.145, radius=3.4564, shapeName=dyn_loinClothBack_r_04_capsule, center=<4.781 -1.3 0>, rotation=<rot 90 180 180>}]
 dyn_loinClothBack_r_MSTR [{shapeType=Capsule, length=1, radius=1, shapeName=dyn_loinClothBack_r_MSTR_capsule}]
 dyn_loinClothFront_l_01 [{shapeType=Capsule, length=2.6, radius=2.3, shapeName=dyn_loinClothFront_l_01_capsule, center=<1.9849 2.1 -2.2>, rotation=<rot 90 0 0>}]
 dyn_loinClothFront_l_02 [{shapeType=Capsule, length=5.25, radius=2.05, shapeName=dyn_loinClothFront_l_02_capsule, center=<3.8 2.1 -2.2>, rotation=<rot 90 0 0>}]
 dyn_loinClothFront_l_03 [{shapeType=Capsule, length=6.6654, radius=3.5262, shapeName=dyn_loinClothFront_l_03_capsule, center=<3.4103 2.9 -2.2>, rotation=<rot 90 0 0>}]
 dyn_loinClothFront_l_04 [{shapeType=Capsule, length=10.8171, radius=2.986, shapeName=dyn_loinClothFront_l_04_capsule, center=<4.4172 1.7 -1.5>, rotation=<rot 90 0 0>}]
 dyn_loinClothFront_l_MSTR [{shapeType=Capsule, length=1, radius=1, shapeName=dyn_loinClothFront_l_MSTR_capsule}]
 dyn_loinClothFront_r_01 [{shapeType=Capsule, length=2.6, radius=2.3, shapeName=dyn_loinClothFront_r_01_capsule, center=<1.9849 2.1 2.2>, rotation=<rot 90 0 0>}]
 dyn_loinClothFront_r_02 [{shapeType=Capsule, length=5.25, radius=2.05, shapeName=dyn_loinClothFront_r_02_capsule, center=<3.8 2.1 2.2>, rotation=<rot 90 0 0>}]
 dyn_loinClothFront_r_03 [{shapeType=Capsule, length=6.65, radius=3.5, shapeName=dyn_loinClothFront_r_03_capsule, center=<3.4103 2.9 2.2>, rotation=<rot 90 0 0>}]
 dyn_loinClothFront_r_04 [{shapeType=Capsule, length=10.8171, radius=2.986, shapeName=dyn_loinClothFront_r_04_capsule, center=<4.4172 1.7 1.5>, rotation=<rot 90 0 0>}]
 dyn_loinClothFront_r_MSTR [{shapeType=Capsule, length=1, radius=1, shapeName=dyn_loinClothFront_r_MSTR_capsule}]
 dyn_strapA_l_01 [{
   shapeType=Capsule
   length=8.2802
   radius=2.1677
   shapeName=dyn_strapA_l_01_capsule
   center=<11.642 -1.2 0.4>
   rotation=<rot 41.399 170.2753 -177.7415>
  }]
 dyn_strapA_l_MSTR [{shapeType=Capsule, length=0.505, radius=0.505, shapeName=dyn_strapA_l_MSTR_capsule, rotation=<rot 0 -90 0>}]
 dyn_strapA_r_01 [{
   shapeType=Capsule
   length=8.072
   radius=2.4352
   shapeName=dyn_strapA_r_01_capsule
   center=<12.1561 -0.8 -1>
   rotation=<rot 9.9656 -98.771 -38.6143>
  }]
 dyn_strapA_r_MSTR [{shapeType=Capsule, length=1, radius=1, shapeName=dyn_strapA_r_MSTR_capsule, rotation=<rot 0 -90 0>}]
 dyn_strapBack_l_01 [{shapeType=Capsule, length=5, radius=2, shapeName=dyn_strapBack_l_01_capsule, center=<3.073 2.283 0>, rotation=<rot 90 0 0>}]
 dyn_strapBack_l_02 [{shapeType=Capsule, length=5, radius=2, shapeName=dyn_strapBack_l_02_capsule, center=<3.073 2.283 0>, rotation=<rot 90 0 0>}]
 dyn_strapBack_l_MSTR [{shapeType=Capsule, length=1, radius=1, shapeName=dyn_strapBack_l_MSTR_capsule}]
 dyn_strapBack_r_01 [{
   shapeType=Capsule
   length=4.4069
   radius=1.5931
   shapeName=dyn_strapBack_r_01_capsule
   center=<2.2884 1.7 0>
   rotation=<rot 90 180 -178.4099>
  }]
 dyn_strapBack_r_02 [{shapeType=Capsule, length=6.989, radius=1.7473, shapeName=dyn_strapBack_r_02_capsule, center=<2.3323 1.4 0>, rotation=<rot 90 0 9.678>}]
 dyn_strapBack_r_MSTR [{shapeType=Capsule, length=1, radius=1, shapeName=dyn_strapBack_r_MSTR_capsule}]
 dyn_strapC_l_01 [{
   shapeType=Capsule
   length=7.9437
   radius=2.561
   shapeName=dyn_strapC_l_01_capsule
   center=<12.2261 0.9 -0.5>
   rotation=<rot 39.2168 -9.6731 -0.6595>
  }]
 dyn_strapC_l_MSTR [{shapeType=Capsule, length=0.505, radius=0.505, shapeName=dyn_strapC_l_MSTR_capsule, rotation=<rot 0 -90 0>}]
 dyn_strapC_r_01 [{
   shapeType=Capsule
   length=10.1462
   radius=1.7429
   shapeName=dyn_strapC_r_01_capsule
   center=<12.5185 1 0.5>
   rotation=<rot -8.3647 -91.9817 27.1323>
  }]
 dyn_strapC_r_MSTR [{shapeType=Capsule, length=1, radius=1, shapeName=dyn_strapC_r_MSTR_capsule, rotation=<rot 0 -90 0>}]
 dyn_strapFront_l_01 [{shapeType=Capsule, length=5.2243, radius=1.7757, shapeName=dyn_strapFront_l_01_capsule, center=<1.7368 1.2 0>, rotation=<rot 90 0 0>}]
 dyn_strapFront_l_02 [{shapeType=Capsule, length=5.1276, radius=1.8724, shapeName=dyn_strapFront_l_02_capsule, center=<1.7368 1.2 0>, rotation=<rot 90 0 0>}]
 dyn_strapFront_l_MSTR [{shapeType=Capsule, length=1, radius=1, shapeName=dyn_strapFront_l_MSTR_capsule}]
 dyn_strapFront_r_01 [{
   shapeType=Capsule
   length=8.5475
   radius=1.4656
   shapeName=dyn_strapFront_r_01_capsule
   center=<3.8176 1.173 0>
   rotation=<rot 90 179.5524 -175.1933>
  }]
 dyn_strapFront_r_02 [{
   shapeType=Capsule
   length=9.3936
   radius=1.3624
   shapeName=dyn_strapFront_r_02_capsule
   center=<2.8392 1.15 -0.3>
   rotation=<rot 83.9457 -3.7952 2.0033>
  }]
 dyn_strapFront_r_MSTR [{shapeType=Capsule, length=1, radius=1, shapeName=dyn_strapFront_r_MSTR_capsule}]
 pelvis [
  {shapeType=Capsule, length=12.6838, radius=10.6966, shapeName=pelvis_capsule, center=<0 -0.9941 0>}
  {shapeType=Capsule, length=11.664, radius=10.5279, shapeName=pelvis_capsule, center=<-3.905 -1.6865 0>}
 ]
 spine_01 [{shapeType=Capsule, length=0.505, radius=0.505, shapeName=spine_01_capsule}]
 thigh_l [{
   shapeType=Capsule
   length=31.4853
   radius=8.6946
   shapeName=thigh_l_capsule
   center=<-14.5502 0.6685 -0.3169>
   rotation=<rot 88.1968 176.6878 -179.227>
  }]
 thigh_r [{
   shapeType=Capsule
   length=30.7353
   radius=9.4446
   shapeName=thigh_l_capsule
   center=<14.5252 -1.019 0.3167>
   rotation=<rot 88.1968 176.6878 0.773>
  }]
}
constraints [
 {
  bone1Name=spine_01
  bone2Name=pelvis
  swing1LimitDegrees=45
  swing1Motion=Free
  swing2LimitDegrees=45
  swing2Motion=Free
  twistLimitDegrees=45
  twistMotion=Free
 }
 {
  bone1Name=thigh_l
  bone2Name=pelvis
  swing1LimitDegrees=45
  swing1Motion=Free
  swing2LimitDegrees=45
  swing2Motion=Free
  twistLimitDegrees=45
  twistMotion=Free
 }
 {bone1Name=dyn_loinClothBack_l_MSTR, bone2Name=pelvis, swing1LimitDegrees=45, swing2LimitDegrees=20, twistLimitDegrees=0}
 {bone1Name=dyn_loinClothBack_l_01, bone2Name=dyn_loinClothBack_l_MSTR, swing1LimitDegrees=20, swing2LimitDegrees=20, twistLimitDegrees=0}
 {bone1Name=dyn_loinClothBack_l_02, bone2Name=dyn_loinClothBack_l_01, swing1LimitDegrees=25, swing2LimitDegrees=20, twistLimitDegrees=0}
 {bone1Name=dyn_loinClothBack_l_03, bone2Name=dyn_loinClothBack_l_02, swing1LimitDegrees=30, swing2LimitDegrees=20, twistLimitDegrees=0}
 {bone1Name=dyn_loinClothBack_l_04, bone2Name=dyn_loinClothBack_l_03, swing1LimitDegrees=40, swing2LimitDegrees=45, twistLimitDegrees=0}
 {bone1Name=dyn_loinClothBack_r_MSTR, bone2Name=pelvis, swing1LimitDegrees=45, swing2LimitDegrees=20, twistLimitDegrees=0}
 {bone1Name=dyn_loinClothBack_r_01, bone2Name=dyn_loinClothBack_r_MSTR, swing1LimitDegrees=20, swing2LimitDegrees=20, twistLimitDegrees=0}
 {bone1Name=dyn_loinClothBack_r_02, bone2Name=dyn_loinClothBack_r_01, swing1LimitDegrees=25, swing2LimitDegrees=20, twistLimitDegrees=0}
 {bone1Name=dyn_loinClothBack_r_03, bone2Name=dyn_loinClothBack_r_02, swing1LimitDegrees=40, swing2LimitDegrees=45, twistLimitDegrees=0}
 {bone1Name=dyn_loinClothBack_r_04, bone2Name=dyn_loinClothBack_r_03, swing1LimitDegrees=40, swing2LimitDegrees=45, twistLimitDegrees=0}
 {bone1Name=dyn_loinClothFront_l_MSTR, bone2Name=pelvis, swing1LimitDegrees=45, swing2LimitDegrees=20, twistLimitDegrees=0}
 {bone1Name=dyn_loinClothFront_l_01, bone2Name=dyn_loinClothFront_l_MSTR, swing1LimitDegrees=45, swing2LimitDegrees=20, twistLimitDegrees=0}
 {bone1Name=dyn_loinClothFront_l_02, bone2Name=dyn_loinClothFront_l_01, swing1LimitDegrees=45, swing2LimitDegrees=20, twistLimitDegrees=0}
 {bone1Name=dyn_loinClothFront_l_03, bone2Name=dyn_loinClothFront_l_02, swing1LimitDegrees=45, swing2LimitDegrees=20, twistLimitDegrees=0}
 {bone1Name=dyn_loinClothFront_l_04, bone2Name=dyn_loinClothFront_l_03, swing1LimitDegrees=45, swing2LimitDegrees=20, twistLimitDegrees=0}
 {bone1Name=dyn_loinClothFront_r_MSTR, bone2Name=pelvis, swing1LimitDegrees=45, swing2LimitDegrees=20, twistLimitDegrees=0}
 {bone1Name=dyn_loinClothFront_r_01, bone2Name=dyn_loinClothFront_r_MSTR, swing1LimitDegrees=45, swing2LimitDegrees=20, twistLimitDegrees=0}
 {bone1Name=dyn_loinClothFront_r_02, bone2Name=dyn_loinClothFront_r_01, swing1LimitDegrees=45, swing2LimitDegrees=20, twistLimitDegrees=0}
 {bone1Name=dyn_loinClothFront_r_03, bone2Name=dyn_loinClothFront_r_02, swing1LimitDegrees=45, swing2LimitDegrees=20, twistLimitDegrees=0}
 {bone1Name=dyn_loinClothFront_r_04, bone2Name=dyn_loinClothFront_r_03, swing1LimitDegrees=45, swing2LimitDegrees=20, twistLimitDegrees=0}
 {
  bone1Name=dyn_strapBack_l_MSTR
  bone2Name=pelvis
  swing1LimitDegrees=45
  swing1Motion=Free
  swing2LimitDegrees=45
  swing2Motion=Free
  twistLimitDegrees=45
  twistMotion=Free
 }
 {bone1Name=dyn_strapBack_l_01, bone2Name=dyn_strapBack_l_MSTR, swing1LimitDegrees=30, swing2LimitDegrees=20, twistLimitDegrees=0}
 {bone1Name=dyn_strapBack_l_02, bone2Name=dyn_strapBack_l_01, swing1LimitDegrees=30, swing2LimitDegrees=10, twistLimitDegrees=0}
 {
  bone1Name=dyn_strapFront_l_MSTR
  bone2Name=pelvis
  swing1LimitDegrees=45
  swing1Motion=Free
  swing2LimitDegrees=45
  swing2Motion=Free
  twistLimitDegrees=45
  twistMotion=Free
 }
 {bone1Name=dyn_strapFront_l_01, bone2Name=dyn_strapFront_l_MSTR, swing1LimitDegrees=60, swing2LimitDegrees=10, twistLimitDegrees=0}
 {bone1Name=dyn_strapFront_l_02, bone2Name=dyn_strapFront_l_01, swing1LimitDegrees=45, swing2LimitDegrees=45, twistLimitDegrees=0}
 {
  bone1Name=dyn_strapBack_r_MSTR
  bone2Name=pelvis
  swing1LimitDegrees=45
  swing1Motion=Free
  swing2LimitDegrees=45
  swing2Motion=Free
  twistLimitDegrees=45
  twistMotion=Free
 }
 {bone1Name=dyn_strapBack_r_01, bone2Name=dyn_strapBack_r_MSTR, swing1LimitDegrees=45, swing2LimitDegrees=20, twistLimitDegrees=0}
 {bone1Name=dyn_strapBack_r_02, bone2Name=dyn_strapBack_r_01, swing1LimitDegrees=45, swing2LimitDegrees=20, twistLimitDegrees=0}
 {
  bone1Name=dyn_strapFront_r_MSTR
  bone2Name=pelvis
  swing1LimitDegrees=45
  swing1Motion=Free
  swing2LimitDegrees=45
  swing2Motion=Free
  twistLimitDegrees=45
  twistMotion=Free
 }
 {bone1Name=dyn_strapFront_r_01, bone2Name=dyn_strapFront_r_MSTR, swing1LimitDegrees=60, swing2LimitDegrees=20, twistLimitDegrees=0}
 {bone1Name=dyn_strapFront_r_02, bone2Name=dyn_strapFront_r_01, swing1LimitDegrees=55, swing2LimitDegrees=20, twistLimitDegrees=0}
 {
  bone1Name=dyn_strapA_l_MSTR
  bone2Name=pelvis
  swing1LimitDegrees=45
  swing1Motion=Free
  swing2LimitDegrees=45
  swing2Motion=Free
  twistLimitDegrees=45
  twistMotion=Free
 }
 {bone1Name=dyn_strapA_l_01, bone2Name=dyn_strapA_l_MSTR, swing1LimitDegrees=45, swing2LimitDegrees=45, twistLimitDegrees=0}
 {
  bone1Name=dyn_strapC_l_MSTR
  bone2Name=pelvis
  swing1LimitDegrees=45
  swing1Motion=Free
  swing2LimitDegrees=45
  swing2Motion=Free
  twistLimitDegrees=45
  twistMotion=Free
 }
 {bone1Name=dyn_strapC_l_01, bone2Name=dyn_strapC_l_MSTR, swing1LimitDegrees=30, swing2LimitDegrees=30, twistLimitDegrees=0}
 {
  bone1Name=dyn_strapA_r_MSTR
  bone2Name=pelvis
  swing1LimitDegrees=45
  swing1Motion=Free
  swing2LimitDegrees=45
  swing2Motion=Free
  twistLimitDegrees=45
  twistMotion=Free
 }
 {bone1Name=dyn_strapA_r_01, bone2Name=dyn_strapA_r_MSTR, swing1LimitDegrees=45, swing2LimitDegrees=45, twistLimitDegrees=0}
 {
  bone1Name=dyn_strapC_r_MSTR
  bone2Name=pelvis
  swing1LimitDegrees=45
  swing1Motion=Free
  swing2LimitDegrees=45
  swing2Motion=Free
  twistLimitDegrees=45
  twistMotion=Free
 }
 {bone1Name=dyn_strapC_r_01, bone2Name=dyn_strapC_r_MSTR, swing1LimitDegrees=45, swing2LimitDegrees=45, twistLimitDegrees=0}
 {
  bone1Name=thigh_r
  bone2Name=pelvis
  swing1LimitDegrees=45
  swing1Motion=Free
  swing2LimitDegrees=45
  swing2Motion=Free
  twistLimitDegrees=45
  twistMotion=Free
 }
]
