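# Animation Blueprint for the Citadel male body cosmetic, driving its straps and loincloth over the pose copied from the attached parent mesh.
# Deformation and strap Control Rigs feed pelvis-space RigidBody sims, with separate physics assets for front end, crouch-walking and other gameplay.
# Boolean blends pick the loincloth result, with bone offsets on the crouch-walk branch, before a spine-weighted chest constraint.
# The update event sets movement flags, gravity override and acceleration scale from skydiving and backbling length.
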AnimBlueprint /BRCosmetics/Characters/Player/Male/Medium/Bodies/M_MED_Citadel/Meshes/M_MED_Citadel_AnimBP
target_skeleton=&/M_MED_Citadel_Skeleton
graphs [
 AnimGraph {nodes [
   "Transform (Modify) Bone - Bone: dyn_strapFront_l_01"!(ComponentPose, bAlphaBoolEnabled:bool=Get bIsCrouching(), Translation="0,0,0", Rotation="0,0,0", Scale="1,1,1") -> (Pose) {
    alphaInputType=Bool
    translationMode=BMM_Additive
    translationSpace=BCS_BoneSpace
    alphaBoolBlend {blendOption=Linear}
    boneToModify {boneName=dyn_strapFront_l_01}
   }
   n1 = "Constraint - chest (PPPPP)"!(
    ComponentPose=Local To Component(LocalPose=Use cached pose 'loinClothCache'())
    Alpha=1
    ConstraintWeights_0=0.12
    ConstraintWeights_1:float=0.45
    ConstraintWeights_2:float=0.3
    ConstraintWeights_3:float=0.1
    ConstraintWeights_4:float=0.03
   ) -> (Pose) {
    alphaBoolBlend {blendOption=Linear}
    boneToModify {boneName=chest}
    constraintSetup [
     {targetBone {boneName=spine_05}}
     {targetBone {boneName=spine_04}}
     {targetBone {boneName=spine_03}}
     {targetBone {boneName=spine_02}}
     {targetBone {boneName=spine_01}}
    ]
    constraintWeights [1,1,1,1,1]
   }
   Output Pose(Result=Component To Local(ComponentPose=@n1)) {layerGroup=None}
   n2 = Control Rig(Source=Copy Pose From Mesh() {bUseAttachedParent=true}) {
    alphaBoolBlend {blendOption=Linear}
    controlRigAssetReference {blueprintRigClass=&/M_MED_Citadel_ControlRig.M_MED_Citadel_ControlRig_C}
   }
   Save cached pose 'DeformationRigCache'(Pose=@n2) {cacheName=DeformationRigCache}
   n3 = Get isCrouchWalking() -> (isCrouchWalking:bool)
   n4 = "Transform (Modify) Bone - Bone: dyn_loinClothFront_l_MSTR"!(
    ComponentPose=Local To Component(LocalPose=Use cached pose 'StrapRigCache'())
    bAlphaBoolEnabled:bool=@n3
    Translation="0,0,-3"
    Rotation="0,0,0"
    Scale="1,1,1"
   ) -> (Pose) {
    alphaInputType=Bool
    rotationMode=BMM_Additive
    rotationSpace=BCS_BoneSpace
    translationMode=BMM_Additive
    translationSpace=BCS_BoneSpace
    alphaBoolBlend {blendInTime=0.3, blendOption=Linear, blendOutTime=1}
    boneToModify {boneName=dyn_loinClothFront_l_MSTR}
    translation {z=-3}
   }
   n5 = "Transform (Modify) Bone - Bone: dyn_loinClothFront_r_MSTR"!(ComponentPose=@n4, bAlphaBoolEnabled:bool=@n3, Translation="0,0,3", Rotation="0,0,0", Scale="1,1,1") -> (Pose) {
    alphaInputType=Bool
    rotationMode=BMM_Additive
    rotationSpace=BCS_BoneSpace
    translationMode=BMM_Additive
    translationSpace=BCS_BoneSpace
    alphaBoolBlend {blendInTime=0.3, blendOption=Linear, blendOutTime=1}
    boneToModify {boneName=dyn_loinClothFront_r_MSTR}
    translation {z=3}
   }
   n6 = Get AccScaleVals() -> (AccScaleVals:Vector)
   n7 = RigidBody!(
    ComponentPose=@n5
    Alpha
    OverrideWorldGravity:Vector="0,0,-980"
    ExternalForce
    ComponentLinearAccScale:Vector=@n6
    ComponentLinearVelScale:Vector=@n6
   ) -> (Pose) {
    bClampLinearTranslationLimitToRefPose=true
    bOverrideWorldGravity=true
    overridePhysicsAsset=&/M_MED_Citadel_CrouchWalk_Physics
    simulationSpace=BaseBoneSpace
    alphaBoolBlend {blendOption=Linear}
    baseBoneRef {boneName=pelvis}
    componentAppliedLinearAccClamp=<6000 6000 6000>
    componentLinearAccScale=<0.4 0.4 0.4>
    componentLinearVelScale=<0.4 0.4 0.4>
   }
   n8 = "Transform (Modify) Bone - Bone: dyn_loinClothBack_r_02"!(ComponentPose=@n7, bAlphaBoolEnabled:bool=True, Translation="0,-2,0", Rotation="0,0,0", Scale="1,1,1") -> (Pose) {
    alphaInputType=Bool
    rotationMode=BMM_Additive
    rotationSpace=BCS_BoneSpace
    translationMode=BMM_Additive
    translationSpace=BCS_BoneSpace
    alphaBoolBlend {blendInTime=0.3, blendOption=Linear, blendOutTime=1}
    boneToModify {boneName=dyn_loinClothBack_r_02}
    translation {y=-2}
   }
   n9 = Component To Local(ComponentPose=@n8)
   n10 = Get AccScaleVals()
   n11 = RigidBody(
    ComponentPose=Local To Component(LocalPose=Use cached pose 'StrapRigCache'())
    OverrideWorldGravity=Get Gravity Override()
    ComponentLinearAccScale=@n10
    ComponentLinearVelScale=@n10
   ) {
    bClampLinearTranslationLimitToRefPose=true
    bOverrideWorldGravity=true
    overridePhysicsAsset=&/M_MED_Citadel_Physics
    simulationSpace=BaseBoneSpace
    alphaBoolBlend {blendOption=Linear}
    baseBoneRef {boneName=pelvis}
    componentAppliedLinearAccClamp=<6000 6000 6000>
    componentLinearAccScale=<0.4 0.4 0.4>
    componentLinearVelScale=<0.4 0.4 0.4>
   }
   n12 = Component To Local(ComponentPose=@n11)
   n13 = Get isCrouchWalking()
   n14 = RigidBody(
    ComponentPose=Local To Component(LocalPose=Use cached pose 'StrapRigCache'())
    OverrideWorldGravity=Get Gravity Override()
    ComponentLinearAccScale="0.4,0.4,0.4"
    ComponentLinearVelScale="0.4,0.4,0.4"
   ) {
    bClampLinearTranslationLimitToRefPose=true
    bOverrideWorldGravity=true
    overridePhysicsAsset=&/M_MED_Citadel_FrontEnd_Physics
    simulationSpace=BaseBoneSpace
    alphaBoolBlend {blendOption=Linear}
    baseBoneRef {boneName=pelvis}
    componentAppliedLinearAccClamp=<6000 6000 6000>
    componentLinearAccScale=<0.4 0.4 0.4>
    componentLinearVelScale=<0.4 0.4 0.4>
   }
   Save cached pose 'loinClothCache'(Pose=Blend Poses by bool(
     BlendPose_0=Component To Local(ComponentPose=@n14)
     BlendPose_1=Blend Poses by bool(BlendPose_0=@n9, BlendPose_1=@n12, bActiveValue=@n13) {blendType=Linear}
     bActiveValue=Get bIsInFrontEnd()
    ) {blendType=Linear}) {cacheName=loinClothCache}
   n15 = Control Rig(Source=Use cached pose 'DeformationRigCache'()) {
    alphaBoolBlend {blendOption=Linear}
    controlRigAssetReference {blueprintRigClass=&/M_MED_CitadelStrap_CtrlRig.M_MED_CitadelStrap_CtrlRig_C}
   }
   Save cached pose 'StrapRigCache'(Pose=@n15) {cacheName=StrapRigCache}
  ]}
 EventGraph {nodes [
   Event BlueprintUpdateAnimation()
   > n1 = Sequence!(execute) -> (then_0, then_1, then_2:Exec, then_3:Exec, then_4:Exec)
   > n2 = Gravity Override State Function DEPRECATED(CharacterPartAnimInstance=Self-Reference(), GameTimeInSeconds=0, GravityOverrideParams=Get Gravity Override Params())
   > Set Gravity Override(Gravity Override=@n2)
   n3 = Branch(execute=@n1.then_1, Condition=Get bIsSkyDiving())
   > Set AccScaleVals(AccScaleVals:Vector="0.1,0.1,0") -> (Output_Get:Vector)
   > n4 = Branch(Condition=Get Is Backbling Long() -> (Is Backbling Long:bool))
   > Set AccScaleVals(AccScaleVals="0.1,0.4,0.4")
   Set AccScaleVals(execute=@n4.else, AccScaleVals="0.3,0.3,0.3")
   Set AccScaleVals(execute=@n3.else, AccScaleVals="0.38,0.6,0.4")
   n5 = PlayerMovement(execute=@n1.then_2, AnimInstance=Self-Reference(), XDotProductCoefficient=1, YDotProductCoefficient=1, ZDotProductCoefficient=1)
   > n6 = Sequence!(execute) -> (then_0, then_1, then_2)
   > n7 = Branch(Condition=OR Boolean!(A="Not Equal (Float)"(A=@n5), B="Not Equal (Float)"(A=@n5.YMovement), C:bool="Not Equal (Float)"(A=@n5.ZMovement)) -> (ReturnValue))
   > Set isWalking(isWalking:bool=true) -> (Output_Get:bool)
   Set isWalking(execute=@n7.else, isWalking=false)
   n8 = "float > float"(A=MapRangeClamped(Value=@n5, InRangeA=-1, InRangeB=0, OutRangeA=1, OutRangeB=0), B=0.5)
   n9 = MapRangeClamped(Value=@n5, InRangeA=0, InRangeB=1)
   n10 = Branch(execute=@n6.then_1, Condition=OR Boolean!(A="float > float"(A=@n9, B=0.5), B=@n8) -> (ReturnValue))
   > Set PlayerStrafe(PlayerStrafe:bool=true) -> (Output_Get:bool)
   Set PlayerStrafe(execute=@n10.else, PlayerStrafe=false)
   n11 = Branch(execute=@n6.then_2, Condition=AND Boolean(A=@n8, B=Get bIsCrouching()))
   > Set isCrouchStrafeLeft(isCrouchStrafeLeft:bool=true) -> (Output_Get:bool)
   Set isCrouchStrafeLeft(execute=@n11.else, isCrouchStrafeLeft=false)
   n12 = Branch(execute=@n1.then_3, Condition=AND Boolean(A=Get bIsCrouching(), B=Get isWalking() -> (isWalking:bool)))
   > Set isCrouchWalking(isCrouchWalking:bool=true) -> (Output_Get:bool)
   Set isCrouchWalking(execute=@n12.else, isCrouchWalking=false)
   Set Is Backbling Long(execute=@n1.then_4, Is Backbling Long:bool=CheckForBackbling().bIsBackblingLong) -> (Output_Get:bool)
   Get PlayerStrafe() -> (PlayerStrafe:bool)
  ]}
]
variables [
 AccScaleVals {type=Vector}
 Gravity Override {type=Vector}
 Gravity Override Params {type=GravityOverrideParamsStruct, default {skydive_GoingUp {gravityOverride=<0 -300 800>}}}
 Is Backbling Long {type=boolean}
 isCrouchStrafeLeft {type=boolean}
 isCrouchWalking {type=boolean}
 isWalking {type=boolean}
 PlayerStrafe {type=boolean}
]
