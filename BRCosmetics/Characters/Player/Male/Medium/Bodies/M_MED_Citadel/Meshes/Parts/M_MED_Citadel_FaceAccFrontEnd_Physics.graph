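# Physics asset for the front-end face accessory mesh of the M_MED_Citadel cosmetic body, set up for dynamic back-hair motion.
# Center, left and right chains of capsule bodies each hang from a master bone constrained to the head with free rotation.
# Links down each chain use tuned swing limits, with twist mostly zero.
# Head, neck, upper spine and clavicle shapes approximate the upper body around the hair.
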PhysicsAsset /BRCosmetics/Characters/Player/Male/Medium/Bodies/M_MED_Citadel/Meshes/Parts/M_MED_Citadel_FaceAccFrontEnd_Physics
bodies {
 clavicle_l [{
   shapeType=Capsule
   length=18.5502
   radius=8.5748
   shapeName=clavicle_l_capsule
   center=<6.4131 -1.7392 -1.0461>
   rotation=<rot 84.1081 -168.7322 -179.9996>
  }]
 clavicle_r [{
   shapeType=Capsule
   length=18.5502
   radius=8.5748
   shapeName=clavicle_l_capsule
   center=<-6.4131 1.7392 1.0461>
   rotation=<rot 84.1081 -168.7322 0.0004>
  }]
 dyn_hairBk_c_01 [{shapeType=Capsule, length=3, radius=2.5, shapeName=dyn_hairBk_c_01_capsule, center=<2.5557 -1.2677 0>}]
 dyn_hairBk_c_02 [{shapeType=Capsule, length=3.5763, radius=1.9237, shapeName=dyn_hairBk_c_02_capsule, center=<2.9156 -1.2677 0>}]
 dyn_hairBk_c_03 [{shapeType=Capsule, length=3.25, radius=2.25, shapeName=dyn_hairBk_c_03_capsule, center=<2.9156 -1.6 0>}]
 dyn_hairBk_c_04 [{shapeType=Capsule, length=3.75, radius=1.75, shapeName=dyn_hairBk_c_04_capsule, center=<2.9156 -1.5071 0>}]
 dyn_hairBk_c_05 [{shapeType=Capsule, length=6.3067, radius=2.4433, shapeName=dyn_hairBk_c_05_capsule, center=<2.8009 -1.5 0>}]
 dyn_hairBk_c_06 [{shapeType=Capsule, length=3.7223, radius=1.7777, shapeName=dyn_hairBk_c_06_capsule, center=<2.8009 0 0>}]
 dyn_hairBk_c_MSTR [{shapeType=Capsule, length=0.5, radius=0.5, shapeName=dyn_hairBk_c_MSTR_capsule}]
 dyn_hairBk_l_01 [{shapeType=Capsule, length=3, radius=2.5, shapeName=dyn_hairBk_l_01_capsule, center=<2.5557 -1.2677 0>}]
 dyn_hairBk_l_02 [{shapeType=Capsule, length=3.8175, radius=1.6, shapeName=dyn_hairBk_l_02_capsule, center=<2.5557 -1.5 0>, rotation=<rot 0 0 31.9167>}]
 dyn_hairBk_l_03 [{shapeType=Capsule, length=4.7025, radius=3.1, shapeName=dyn_hairBk_l_03_capsule, center=<2.5557 -3 0>, rotation=<rot 0 0 33.9107>}]
 dyn_hairBk_l_04 [
  {shapeType=Capsule, length=8.1854, radius=2.5223, shapeName=dyn_hairBk_l_04_capsule, center=<2.5557 -1.5 0>, rotation=<rot 0 0.0001 21.9567>}
  {
   shapeType=Capsule
   length=5.4236
   radius=2.2618
   shapeName=dyn_hairBk_l_04_capsule
   center=<1.8317 1.5253 2.7755>
   rotation=<rot 0 0.0001 39.6476>
  }
 ]
 dyn_hairBk_l_05 [
  {shapeType=Capsule, length=5.875, radius=2.4, shapeName=dyn_hairBk_l_05_capsule, center=<2.5557 -1.5 0>}
  {shapeType=Capsule, length=5.875, radius=2.4, shapeName=dyn_hairBk_l_05_capsule, center=<2.2936 0.3111 0.9817>, rotation=<rot 0 0 17.5822>}
 ]
 dyn_hairBk_l_MSTR [{shapeType=Capsule, length=0.5, radius=0.5, shapeName=dyn_hairBk_l_MSTR_capsule}]
 dyn_hairBk_r_01 [{shapeType=Capsule, length=4.2054, radius=1.2946, shapeName=dyn_hairBk_r_01_capsule, center=<2.5557 -1.2677 0>, rotation=<rot 0 0 -42.7113>}]
 dyn_hairBk_r_02 [{shapeType=Capsule, length=4.879, radius=1.7, shapeName=dyn_hairBk_r_02_capsule, center=<2.5557 -1.5 0>, rotation=<rot 0 0.0001 -40.0811>}]
 dyn_hairBk_r_03 [{shapeType=Capsule, length=5.4641, radius=2.5, shapeName=dyn_hairBk_r_03_capsule, center=<2.5557 -3 0>, rotation=<rot 0 0.0001 -16.2185>}]
 dyn_hairBk_r_04 [{shapeType=Capsule, length=7.2178, radius=1.8, shapeName=dyn_hairBk_r_04_capsule, center=<2.5557 -1.5 0>, rotation=<rot 0 0 -23.535>}]
 dyn_hairBk_r_05 [{shapeType=Capsule, length=4.2894, radius=2.5638, shapeName=dyn_hairBk_r_05_capsule, center=<2.5557 -1.5 0>, rotation=<rot 0 0.0001 -13.8704>}]
 dyn_hairBk_r_MSTR [{shapeType=Capsule, length=0.5, radius=0.5, shapeName=dyn_hairBk_r_MSTR_capsule}]
 head [
  {shapeType=Capsule, length=5.6634, radius=5.686, center=<4.9817 -0.2 0>, rotation=<rot 90 0 0>}
  {shapeType=Capsule, length=8.3234, radius=6.0485, shapeName=head_capsule, center=<3.4471 3.7892 0>, rotation=<rot 90 0 16.2417>}
 ]
 neck_02 [
  {shapeType=Capsule, length=10.1453, radius=5.3746, shapeName=neck_02_capsule, center=<1.8504 -1.0489 0>, rotation=<rot 90 0 0>}
  {shapeType=Capsule, length=0, radius=1.0423, shapeName=neck_02_capsule}
  {shapeType=Capsule, length=0, radius=2.0306, shapeName=neck_02_capsule, center=<-2.5793 0 0>}
  {shapeType=Capsule, length=38.2112, radius=4.4271, shapeName=neck_02_capsule, center=<2.3788 2.6989 0>}
  {shapeType=Capsule, length=37.0118, radius=5.6265, shapeName=neck_02_capsule, center=<-1.9099 3.4446 0>}
 ]
 spine_05 [
  {shapeType=Sphere, radius=10.2041, shapeName=spine_05_sphere, center=<6.5311 -0.1772 0>}
  {shapeType=Capsule, length=15.805, radius=11.3472, shapeName=spine_05_capsule, center=<-0.4404 0 0>, rotation=<rot 0 -90 0>}
  {shapeType=Capsule, length=5.8484, radius=9.2907, shapeName=spine_05_capsule, center=<6.5912 0.0421 0>}
 ]
}
constraints [
 {
  bone1Name=dyn_hairBk_c_MSTR
  bone2Name=head
  swing1LimitDegrees=45
  swing1Motion=Free
  swing2LimitDegrees=45
  swing2Motion=Free
  twistLimitDegrees=45
  twistMotion=Free
 }
 {bone1Name=dyn_hairBk_c_01, bone2Name=dyn_hairBk_c_MSTR, swing1LimitDegrees=40, swing2LimitDegrees=20, twistLimitDegrees=0}
 {bone1Name=dyn_hairBk_c_02, bone2Name=dyn_hairBk_c_01, swing1LimitDegrees=35, swing2LimitDegrees=0, twistLimitDegrees=0}
 {bone1Name=dyn_hairBk_c_03, bone2Name=dyn_hairBk_c_02, swing1LimitDegrees=35, swing2LimitDegrees=5, twistLimitDegrees=0}
 {bone1Name=dyn_hairBk_c_04, bone2Name=dyn_hairBk_c_03, swing1LimitDegrees=25, swing2LimitDegrees=12, twistLimitDegrees=20}
 {bone1Name=dyn_hairBk_c_05, bone2Name=dyn_hairBk_c_04, swing1LimitDegrees=16, swing2LimitDegrees=8, twistLimitDegrees=20}
 {bone1Name=dyn_hairBk_c_06, bone2Name=dyn_hairBk_c_05, swing1LimitDegrees=10, swing2LimitDegrees=12, twistLimitDegrees=20}
 {
  bone1Name=dyn_hairBk_l_MSTR
  bone2Name=head
  swing1LimitDegrees=45
  swing1Motion=Free
  swing2LimitDegrees=45
  swing2Motion=Free
  twistLimitDegrees=45
  twistMotion=Free
 }
 {bone1Name=dyn_hairBk_l_01, bone2Name=dyn_hairBk_l_MSTR, swing1LimitDegrees=15, swing2LimitDegrees=30, twistLimitDegrees=0}
 {bone1Name=dyn_hairBk_l_02, bone2Name=dyn_hairBk_l_01, swing1LimitDegrees=30, swing2LimitDegrees=0, twistLimitDegrees=0}
 {bone1Name=dyn_hairBk_l_03, bone2Name=dyn_hairBk_l_02, swing1LimitDegrees=30, swing2LimitDegrees=5, twistLimitDegrees=0}
 {bone1Name=dyn_hairBk_l_04, bone2Name=dyn_hairBk_l_03, swing1LimitDegrees=30, swing2LimitDegrees=15, twistLimitDegrees=0}
 {bone1Name=dyn_hairBk_l_05, bone2Name=dyn_hairBk_l_04, swing1LimitDegrees=30, swing2LimitDegrees=15, twistLimitDegrees=0}
 {
  bone1Name=dyn_hairBk_r_MSTR
  bone2Name=head
  swing1LimitDegrees=45
  swing1Motion=Free
  swing2LimitDegrees=45
  swing2Motion=Free
  twistLimitDegrees=45
  twistMotion=Free
 }
 {bone1Name=dyn_hairBk_r_01, bone2Name=dyn_hairBk_r_MSTR, swing1LimitDegrees=30, swing2LimitDegrees=10, twistLimitDegrees=0}
 {bone1Name=dyn_hairBk_r_02, bone2Name=dyn_hairBk_r_01, swing1LimitDegrees=45, swing2LimitDegrees=0, twistLimitDegrees=0}
 {bone1Name=dyn_hairBk_r_03, bone2Name=dyn_hairBk_r_02, swing1LimitDegrees=30, swing2LimitDegrees=15, twistLimitDegrees=0}
 {bone1Name=dyn_hairBk_r_04, bone2Name=dyn_hairBk_r_03, swing1LimitDegrees=30, swing2LimitDegrees=15, twistLimitDegrees=0}
 {bone1Name=dyn_hairBk_r_05, bone2Name=dyn_hairBk_r_04, swing1LimitDegrees=30, swing2LimitDegrees=15, twistLimitDegrees=0}
 {
  bone1Name=head
  bone2Name=neck_02
  swing1LimitDegrees=45
  swing1Motion=Free
  swing2LimitDegrees=45
  swing2Motion=Free
  twistLimitDegrees=45
  twistMotion=Free
 }
 {
  bone1Name=neck_02
  bone2Name=spine_05
  swing1LimitDegrees=45
  swing1Motion=Free
  swing2LimitDegrees=45
  swing2Motion=Free
  twistLimitDegrees=45
  twistMotion=Free
 }
 {
  bone1Name=clavicle_r
  bone2Name=spine_05
  swing1LimitDegrees=45
  swing1Motion=Free
  swing2LimitDegrees=45
  swing2Motion=Free
  twistLimitDegrees=45
  twistMotion=Free
 }
 {
  bone1Name=clavicle_l
  bone2Name=spine_05
  swing1LimitDegrees=45
  swing1Motion=Free
  swing2LimitDegrees=45
  swing2Motion=Free
  twistLimitDegrees=45
  twistMotion=Free
 }
]
