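# Anim Blueprint for the M_MED_Citadel face accessory that simulates front and back hair physics on the character.
# It copies the parent mesh pose, simulates front hair with RigidBody into a cached pose, then blends between back-hair physics assets for custom cape, backpack, front end or default.
# The backpack branch also offsets hair root bones.
# The EventGraph derives gravity, backbling and movement flags; the skydive bone-rotation chain never reaches the output.
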AnimBlueprint /BRCosmetics/Characters/Player/Male/Medium/Bodies/M_MED_Citadel/Meshes/Parts/M_MED_Citadel_FaceAcc_AnimBP
target_skeleton=&/M_MED_Citadel_FaceAcc_Skeleton
graphs [
 AnimGraph {nodes [
   n1 = Get bIsSkydiveFloating()
   n2 = "Transform (Modify) Bone - Bone: dyn_hairBk_c_03"!(ComponentPose, bAlphaBoolEnabled:bool=@n1, Translation="0,0,0", Rotation="0,40,0", Scale="1,1,1") -> (Pose) {
    alphaInputType=Bool
    rotationMode=BMM_Additive
    rotationSpace=BCS_BoneSpace
    alphaBoolBlend {blendOption=Linear}
    boneToModify {boneName=dyn_hairBk_c_03}
    rotation {yaw=40}
   }
   n3 = "Transform (Modify) Bone - Bone: dyn_hairBk_c_05"!(ComponentPose=@n2, bAlphaBoolEnabled:bool=@n1, Translation="0,0,0", Rotation="0,20,0", Scale="1,1,1") -> (Pose) {
    alphaInputType=Bool
    rotationMode=BMM_Additive
    rotationSpace=BCS_BoneSpace
    alphaBoolBlend {blendOption=Linear}
    boneToModify {boneName=dyn_hairBk_c_05}
    rotation {yaw=20}
   }
   "Transform (Modify) Bone - Bone: dyn_hairBk_r_02"!(ComponentPose=@n3, bAlphaBoolEnabled:bool=@n1, Translation="0,0,0", Rotation="20,0,0", Scale="1,1,1") -> (Pose) {
    alphaInputType=Bool
    rotationMode=BMM_Additive
    rotationSpace=BCS_BoneSpace
    alphaBoolBlend {blendOption=Linear}
    boneToModify {boneName=dyn_hairBk_r_02}
    rotation {pitch=20}
   }
   n4 = RigidBody!(
    ComponentPose=Local To Component(LocalPose=Use cached pose 'hairFrontCache'())
    Alpha
    OverrideWorldGravity:Vector=Get Gravity OverrideBackHair() -> (Gravity OverrideBackHair:Vector)
    ExternalForce
   ) -> (Pose) {
    bClampLinearTranslationLimitToRefPose=true
    bOverrideWorldGravity=true
    overridePhysicsAsset=&/M_MED_Citadel_FaceAccFrontEnd_Physics
    simulationSpace=BaseBoneSpace
    alphaBoolBlend {blendOption=Linear}
    baseBoneRef {boneName=spine_05}
    componentAppliedLinearAccClamp=<6000 6000 6000>
    componentLinearAccScale=<0.4 0.3 0.3>
    componentLinearVelScale=<0.4 0.3 0.3>
    overrideWorldGravity {z=-980}
   }
   n5 = RigidBody!(ComponentPose=Local To Component(LocalPose=Use cached pose 'hairFrontCache'()), Alpha, OverrideWorldGravity=Get Gravity OverrideBackHair()) -> (Pose) {
    bClampLinearTranslationLimitToRefPose=true
    bOverrideWorldGravity=true
    overridePhysicsAsset=&/M_MED_Citadel_FaceAcc_Physics
    simulationSpace=BaseBoneSpace
    alphaBoolBlend {blendOption=Linear}
    baseBoneRef {boneName=spine_05}
    componentAppliedLinearAccClamp=<6000 6000 6000>
    componentLinearAccScale=<0.4 0.3 0.3>
    componentLinearVelScale=<0.4 0.3 0.3>
    overrideWorldGravity {z=-980}
   }
   n6 = Blend Poses by bool(BlendPose_0=Component To Local(ComponentPose=@n4), BlendPose_1=Component To Local(ComponentPose=@n5), bActiveValue=Get bIsInFrontEnd()) {blendType=Linear}
   n7 = RigidBody!(
    ComponentPose=Local To Component(LocalPose=Use cached pose 'hairFrontCache'())
    Alpha
    OverrideWorldGravity="0,0,-980"
    ExternalForce:Vector
    ComponentLinearAccScale:Vector="0,0,0"
    ComponentLinearVelScale:Vector="0,0,0"
   ) -> (Pose) {
    bClampLinearTranslationLimitToRefPose=true
    bOverrideWorldGravity=true
    overridePhysicsAsset=&/M_MED_Citadel_BackPack_FaceAcc_Physics
    simulationSpace=BaseBoneSpace
    alphaBoolBlend {blendOption=Linear}
    baseBoneRef {boneName=spine_05}
    componentAppliedLinearAccClamp=<6000 6000 6000>
    componentLinearAccScale=<0.4 0.3 0.3>
    componentLinearVelScale=<0.4 0.3 0.3>
    overrideWorldGravity {z=-980}
   }
   n8 = "Transform (Modify) Bone - Bone: dyn_hairBk_c_MSTR"!(ComponentPose=@n7, bAlphaBoolEnabled:bool=True, Translation="-1,0,0", Rotation="0,0,0", Scale="1,1,1") -> (Pose) {
    alphaInputType=Bool
    rotationMode=BMM_Additive
    rotationSpace=BCS_BoneSpace
    translationMode=BMM_Additive
    translationSpace=BCS_BoneSpace
    alphaBoolBlend {blendOption=Linear}
    boneToModify {boneName=dyn_hairBk_c_MSTR}
    translation {x=-1}
   }
   n9 = "Transform (Modify) Bone - Bone: dyn_hairBk_l_MSTR"!(ComponentPose=@n8, bAlphaBoolEnabled:bool=True, Translation="-1.5,0,0", Rotation="0,0,0", Scale="1,1,1") -> (Pose) {
    alphaInputType=Bool
    rotationMode=BMM_Additive
    rotationSpace=BCS_BoneSpace
    translationMode=BMM_Additive
    translationSpace=BCS_BoneSpace
    alphaBoolBlend {blendOption=Linear}
    boneToModify {boneName=dyn_hairBk_l_MSTR}
    translation {x=-1.5}
   }
   n10 = "Transform (Modify) Bone - Bone: dyn_hairBk_r_MSTR"!(ComponentPose=@n9, bAlphaBoolEnabled:bool=True, Translation="-1.5,0,0", Rotation="0,0,0", Scale="1,1,1") -> (Pose) {
    alphaInputType=Bool
    rotationMode=BMM_Additive
    rotationSpace=BCS_BoneSpace
    translationMode=BMM_Additive
    translationSpace=BCS_BoneSpace
    alphaBoolBlend {blendOption=Linear}
    boneToModify {boneName=dyn_hairBk_r_MSTR}
    translation {x=-1.5}
   }
   n11 = Component To Local(ComponentPose=@n10)
   n12 = Get Is Wearing BackPack() -> (Is Wearing BackPack:bool)
   n13 = Get AccScaleBackBling() -> (AccScaleBackBling:Vector)
   n14 = RigidBody(
    ComponentPose=Local To Component(LocalPose=Use cached pose 'hairFrontCache'())
    OverrideWorldGravity=Get Gravity OverrideBackHair()
    ComponentLinearAccScale=@n13
    ComponentLinearVelScale=@n13
   ) {
    bClampLinearTranslationLimitToRefPose=true
    bOverrideWorldGravity=true
    overridePhysicsAsset=&/M_MED_Citadel_Cape_FaceAcc_Physics
    simulationSpace=BaseBoneSpace
    alphaBoolBlend {blendOption=Linear}
    baseBoneRef {boneName=spine_05}
    componentAppliedLinearAccClamp=<3000 3000 3000>
    componentLinearAccScale=<0.4 0.3 0.3>
    componentLinearVelScale=<0.4 0.3 0.3>
    overrideWorldGravity {z=-980}
   }
   Output Pose(Result=Blend Poses by bool(
     BlendPose_0=Component To Local(ComponentPose=@n14)
     BlendPose_1=Blend Poses by bool(BlendPose_0=@n11, BlendPose_1=@n6, bActiveValue=@n12) {blendType=Linear}
     bActiveValue=Get isCustomCape() -> (isCustomCape:bool)
    ) {blendType=Linear}) {layerGroup=None}
   n15 = RigidBody(
    ComponentPose=Local To Component(LocalPose=Copy Pose From Mesh() {bUseAttachedParent=true})
    OverrideWorldGravity=Get Gravity Override()
    ComponentLinearAccScale="0.4,0.3,0.3"
    ComponentLinearVelScale="0.4,0.3,0.3"
   ) {
    bClampLinearTranslationLimitToRefPose=true
    bOverrideWorldGravity=true
    overridePhysicsAsset=&/M_MED_Citadel_HairFront_FaceAcc_Physics
    simulationSpace=BaseBoneSpace
    alphaBoolBlend {blendOption=Linear}
    baseBoneRef {boneName=spine_05}
    componentAppliedLinearAccClamp=<6000 6000 6000>
    componentLinearAccScale=<0.4 0.3 0.3>
    componentLinearVelScale=<0.4 0.3 0.3>
    overrideWorldGravity {z=-980}
   }
   Save cached pose 'hairFrontCache'(Pose=Component To Local(ComponentPose=@n15)) {cacheName=hairFrontCache}
   Get AccScaleBackBling()
   Get AccScaleBackBling()
   Get Gravity Override()
   Get AccScaleBackBling()
   Get AccScaleBackBling()
  ]}
 EventGraph {nodes [
   n1 = Event BlueprintUpdateAnimation()
   > n2 = Sequence!(execute) -> (then_0, then_1, then_2:Exec, then_3:Exec, then_4:Exec, then_5:Exec, then_6:Exec, then_7:Exec)
   > n3 = Gravity Override State Function DEPRECATED(CharacterPartAnimInstance=Self-Reference(), GameTimeInSeconds=@n1.DeltaTimeX, GravityOverrideParams=Get Gravity Override Params())
   > n4 = Set Gravity Override(Gravity Override=@n3)
   > n5 = Branch(Condition=Get Is Wearing BackPack())
   > Set Gravity OverrideBackHair(Gravity OverrideBackHair:Vector="0,0,-980") -> (Output_Get:Vector)
   Set Gravity OverrideBackHair(execute=@n5.else, Gravity OverrideBackHair=@n4)
   n6 = PlayerMovement(execute=@n2.then_1, AnimInstance=Self-Reference(), XDotProductCoefficient=1, YDotProductCoefficient=1, ZDotProductCoefficient=1)
   > n7 = Sequence!(execute) -> (then_0, then_1, then_2)
   > n8 = Branch(Condition=OR Boolean!(A="Not Equal (Float)"(A=@n6), B="Not Equal (Float)"(A=@n6.YMovement), C:bool="Not Equal (Float)"(A=@n6.ZMovement)) -> (ReturnValue))
   > Set isWalking(isWalking:bool=true) -> (Output_Get:bool)
   Set isWalking(execute=@n8.else, isWalking=false)
   n9 = "float > float"(A=MapRangeClamped(Value=@n6, InRangeA=-1, InRangeB=0, OutRangeA=1, OutRangeB=0), B=0.5)
   n10 = MapRangeClamped(Value=@n6, InRangeA=0, InRangeB=1)
   n11 = Branch(execute=@n7.then_1, Condition=OR Boolean!(A="float > float"(A=@n10, B=0.5), B=@n9) -> (ReturnValue))
   > Set PlayerStrafe(PlayerStrafe:bool=true) -> (Output_Get:bool)
   Set PlayerStrafe(execute=@n11.else, PlayerStrafe=false)
   n12 = Branch(execute=@n7.then_2, Condition=AND Boolean(A=@n9, B=Get bIsCrouching()))
   > Set isCrouchStrafeLeft(isCrouchStrafeLeft:bool=true) -> (Output_Get:bool)
   Set isCrouchStrafeLeft(execute=@n12.else, isCrouchStrafeLeft=false)
   Set bodyMesh(execute=@n2.then_2, bodyMesh:Skeletal Mesh Component Object Reference=GetPartSkeletalMeshComponent()) -> (Output_Get:Skeletal Mesh Component Object Reference)
   n13 = Branch(execute=@n2.then_5, Condition=AND Boolean(A=Get bIsCrouching(), B=Get isWalking() -> (isWalking:bool)))
   > Set isCrouchWalking(isCrouchWalking:bool=true) -> (Output_Get:bool)
   Set isCrouchWalking(execute=@n13.else, isCrouchWalking=false)
   n14 = Sequence!(execute=@n2.then_7) -> (then_0, then_1, then_2, then_3:Exec, then_4:Exec, then_5:Exec, then_6:Exec)
   n15 = Get isCustomCape()
   n16 = CheckForBackbling()
   n17 = AND Boolean!(A=@n16, B=NOT Boolean(A=@n16.bIsBackblingACape), C:bool=NOT Boolean(A=@n16.bIsBackblingLong)) -> (ReturnValue)
   n18 = Branch(execute=@n14.then_2, Condition=AND Boolean!(A=@n17, B=NOT Boolean(A=@n15)) -> (ReturnValue))
   > Set AccScaleBackBling(AccScaleBackBling:Vector="0.1,0.02,0.1") -> (Output_Get:Vector)
   > Set Is Wearing BackPack(Is Wearing BackPack:bool=true) -> (Output_Get:bool)
   Set AccScaleBackBling(execute=@n18.else, AccScaleBackBling="0.35,0.5,0.1")
   > Set Is Wearing BackPack(Is Wearing BackPack=false)
   n19 = Branch(execute=@n14.then_3, Condition=AND Boolean(A=NOT Boolean(A=@n16), B=Get bIsSkydiveFloating()))
   > Set isSkydivingNotBackbling(isSkydivingNotBackbling:bool=true) -> (Output_Get:bool)
   Set isSkydivingNotBackbling(execute=@n19.else, isSkydivingNotBackbling=false)
   n20 = Branch(execute=@n14.then_4, Condition=AND Boolean(A=@n17, B=Get bIsSkyDiving()))
   > Set isSKydivingHasBackbling(isSKydivingHasBackbling:bool=true) -> (Output_Get:bool)
   Set isSKydivingHasBackbling(execute=@n20.else, isSKydivingHasBackbling=false)
   n21 = Branch(
    execute=@n14.then_5
    Condition=BackpackItemDefinitionHasAnyMetaTag!(Tags:Array of Gameplay Tag Structures=Make Array("[0]"=(TagName="Cosmetics.MetaTags.Citadel.Default.Cape")), bRequiresEquipped:bool=false) -> (ReturnValue:bool)
   )
   > Set isCustomCape(isCustomCape:bool=true) -> (Output_Get:bool)
   Set isCustomCape(execute=@n21.else, isCustomCape=false)
   Set Is Backbling A Cape(execute=@n14.then_6, Is Backbling A Cape:bool=OR Boolean(A=@n16.bIsBackblingACape, B=@n16.bIsBackblingLong)) -> (Output_Get:bool)
   Get bIsSkyDiving()
  ]}
]
variables [
 AccScaleBackBling {type=Vector}
 bodyMesh {type=SkeletalMeshComponent}
 Character Part Anim Instance {type=CustomCharacterPartAnimInstance, default=null}
 Gravity Override {type=Vector}
 Gravity Override Params {type=GravityOverrideParamsStruct}
 Gravity OverrideBackHair {type=Vector}
 headRotDn {type=number}
 headRotUp {type=number}
 Is Backbling A Cape {type=boolean}
 Is Wearing BackPack {type=boolean}
 isCrouchStrafeLeft {type=boolean}
 isCrouchWalking {type=boolean}
 isCustomCape {type=boolean}
 isSKydivingHasBackbling {type=boolean}
 isSkydivingNotBackbling {type=boolean}
 isWalking {type=boolean}
 isWeaponRifleCrouchin {type=boolean}
 isWearingBackbling {type=boolean}
 Pawn Velocity Z {type=number}
 PlayerStrafe {type=boolean}
]
